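# Niagara stage-show effect for the acoustic guitar emote, surrounding the player with lights, sparkles and pyro.
# Spotlight, dome and rotating beam meshes plus a light cycle colors on curves, with audio amplitude blending Local/Other alpha for spot, dome and light.
# Paired side fountains combine GPU sparks, colliding CPU sparks and glowing light sources.
# Front-end modifiers add a dark backdrop, dim pyro and boost spotlights; time of day scales spark alpha.
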NiagaraSystem /BRCosmetics/Animation/Game/MainPlayer/Emotes/Ordinary/FX/NS_Emote_Ordinary_AccusticGuitar
dependencies {
 usedDataInterfaces [
  &NiagaraDataInterfaceArrayFloat3
  &NiagaraDataInterfaceColorCurve
  &NiagaraDataInterfaceCurve
  &NiagaraDataInterfaceArrayFloat
 ]
 usedDynamicInputs [&RandomRangeVector2D]
 usedModules [&SystemState]
}
emitters {
 BlueLight_MESH001 {
  emitterUpdateScript [
   &EmitterState(
    Inactive Response=::NewEnumerator1
    Life Cycle Mode=::NewEnumerator1
    Loop Duration=0.75
    Recalculate Duration Each Loop=true
    UseLoopDelay=true
   )
   &SpawnRate(SpawnRate=&RandomRangeFloat(Minimum=5, Maximum=4), Use Spawn Probability=false)
  ]
  particleSpawnScript [
   &FN_SetSpawnGroup(Number of Spawn Groups=6)
   &InitializeParticle(
    Lifetime Max=0.6
    Lifetime Min=0.8
    Lifetime Mode=::NewEnumerator1
    Mesh Scale=&Multiply_Vector(
     A=&SelectVectorFromArray(
      Direct Array Index=$Particles.SpawnGroup
      Vector Selection Array=DataInterface() {propertyValues {FloatData [<0.1 0.1 1.5>, <0.15 0.15 1.7>, <0.05 0.05 1.9>, <0.15 0.15 1.85>, <0.07 0.07 2>, <0.15 0.15 1.7>]}}
     )
     B=<1 1 0.75>
    )
    Mesh Scale Mode=::NewEnumerator4
    Position Offset=&Multiply_Vector(
     A=&SelectVectorFromArray(
      Direct Array Index=$Particles.SpawnGroup
      Vector Selection Array=DataInterface() {propertyValues {FloatData [<30 20 0>, <30 -20 0>, <0 35 0>, <0 -35 0>, <-40 20 0>, <-40 -20 0>]}}
     )
     B=<0.65 0.65 1>
    )
    UsePositionOffset=true
    Write Material Random=true
   )
   &InitialMeshOrientation(
    Mesh Orientation Mode=::NewEnumerator2
    Rotation=&SelectVectorFromArray(
     Direct Array Index=$Particles.SpawnGroup
     Vector Selection Array=DataInterface() {propertyValues {FloatData [<0.3 0.5 -0.1>, <0.3 0.5 0.1>, <0 0.5 -0.1>, <0 0.5 0.1>, <0 0.45 -0.05>, <0 0.45 0.05>]}}
    )
    Use Rotation Vector=true
   )
  ]
  particleUpdateScript [
   &ParticleState()
   &Color(Color=&HueShiftLinearColor(
     Hue=0.1
     Saturation=1
     Value=1
     Linear Color=&ColorFromCurve(
      ColorCurve=DataInterface() {propertyValues {
        AlphaCurve {keys [<at 0.1138 0.5 RCIM_Cubic>]}
        BlueCurve {keys [<at 0.3144 0 RCIM_Cubic>,<at 0.3778 0.0417>,<at 0.65 0.0417>,<at 0.7 0>]}
        GreenCurve {keys [<at 0.3144 0 RCIM_Cubic>,<at 0.3778 0.5>,<at 0.65 0.5>,<at 0.7 0>]}
        RedCurve {keys [<at 0.3144 0.5 RCIM_Cubic>,<at 0.3778 0>,<at 0.65 0>,<at 0.7 0.5>]}
       }}
      CurveIndex=$Emitter.NormalizedLoopAge
     )
    )) {module=Color001}
   &ScaleMeshSize(Scale Factor=&MakeVector(
     X=1
     Y=1
     Z=&FloatFromCurve(
      FloatCurve=DataInterface() {propertyValues {Curve {keys [
          {interpMode=RCIM_Cubic, leaveTangent=1.3994, tangentMode=RCTM_Break, time=0, value=0.5}
          {interpMode=RCIM_Cubic, tangentMode=RCTM_Break, time=1, value=1}
         ]}}}
      Scale Curve=0.4
     )
    ))
   &UpdateMeshOrientation(
    Rotation Coordinate Space=::NewEnumerator0
    Rotation Rate=&Multiply_Float(
     A=&SelectFloatFromArray(Direct Array Index=$Particles.SpawnGroup, Float Selection Array=DataInterface() {propertyValues {FloatData [0.5,-0.6,0.45,-0.48,0.6,-0.55]}})
     B=0.5
    )
    Rotation Vector=<0 0 1>
   )
  ]
  rendererClasses [&NiagaraMeshRendererProperties, &NiagaraSpriteRendererProperties]
  renderers [
   {rendererClass=&NiagaraMeshRendererProperties, rendererIndex=0}
   {rendererClass=&NiagaraSpriteRendererProperties, rendererIndex=1}
  ]
 }
 DarkBackGround_FE {
  emitterUpdateScript [
   &EmitterState(
    Inactive Response=::NewEnumerator1
    Life Cycle Mode=::NewEnumerator1
    Loop Behavior=::NewEnumerator1
    Loop Duration Mode=::NewEnumerator1
    UseLoopDelay=false
   )
   &SpawnBurst_Instantaneous(Use Spawn Probability=false, UseLoopCountLimit=false)
  ]
  particleSpawnScript [
   &FN_SetSpawnGroup()
   &InitializeParticle(
    Color=<rgba 0 0 0 1>
    Position Offset=<0 0 80>
    Sprite Size=<300 500>
    Sprite Size Mode=::NewEnumerator4
    UsePositionOffset=true
    Write Material Random=true
   )
   &CameraOffset(Camera Offset Amount=-40)
  ]
  particleUpdateScript [
   &ParticleState(
    Kill Particles When Lifetime Has Elapsed=false
    Let Infinitely Lived Particles Die When Emitter Deactivates=true
    Loop Particles Lifetime=true
   )
   &ScaleColor(Scale Alpha=&FN_FrontEndFloatModifier(Gameplay=0, Front End=1), ScaleA=true, ScaleRGB=true)
  ]
  rendererClasses [&NiagaraSpriteRendererProperties]
  renderers [{rendererClass=&NiagaraSpriteRendererProperties}]
 }
 DomeLight_01_LWE {rendererClasses [], renderers []}
 DomeLight_02_LWE {rendererClasses [], renderers []}
 DomeLight_03_LWE {rendererClasses [], renderers []}
 Firework_Burst004 {
  simTarget=GPUComputeSim
  emitterUpdateScript [
   &EmitterState(Delay First Loop Only=true, Life Cycle Mode=::NewEnumerator1, Loop Delay=0.2, Loop Duration=2, UseLoopDelay=true)
   &SpawnBurst_Instantaneous(Spawn Count=150, Use Spawn Probability=false, UseLoopCountLimit=false)
   &SpawnRate(
    SpawnRate=&FloatFromCurve(
     FloatCurve=DataInterface() {propertyValues {Curve {keys [
         {arriveTangent=-14.6052, interpMode=RCIM_Cubic, leaveTangent=-14.6052, tangentMode=RCTM_User, time=0.2, value=1}
         {interpMode=RCIM_Cubic, leaveTangent=-1.9831, time=0.3, value=0}
        ]}}}
     CurveIndex=$Emitter.NormalizedLoopAge
     Scale Curve=250
    )
    Use Spawn Probability=false
   )
   /SetVariables(Emitter.Rand A:NiagaraFloat=&RandomRangeFloat(Minimum=5, Maximum=25)) {isSetParameters=true}
  ]
  particleSpawnScript [
   &FN_SetSpawnGroup(Number of Spawn Groups=2)
   &InitializeParticle(
    Color=<rgba 200 100.8333 30 1>
    Lifetime=&Lerp_Float(
     A=0.1
     B=0.6
     Alpha=&FloatFromCurve(
      FloatCurve=DataInterface() {propertyValues {Curve {keys [
          {
           arriveTangentWeight=1.0235
           interpMode=RCIM_Cubic
           leaveTangentWeight=1.0235
           tangentMode=RCTM_User
           tangentWeightMode=RCTWM_WeightedBoth
           time=0
           value=0
          }
          {
           arriveTangent=7.6705
           arriveTangentWeight=0.8322
           interpMode=RCIM_Cubic
           leaveTangent=7.6705
           leaveTangentWeight=0.8322
           tangentMode=RCTM_User
           tangentWeightMode=RCTWM_WeightedBoth
           time=0.9
           value=1
          }
         ]}}}
      CurveIndex=&RandomRangeFloat(Minimum=0, Maximum=1)
     )
    )
    Sprite Size Max=<0.5 3>
    Sprite Size Min=<1.5 8>
    Sprite Size Mode=::NewEnumerator2
    UsePositionOffset=false
    Write Material Random=true
   )
   &ShapeLocation(MaskSpawnGroup=false, Override Seed=false, Sphere Radius=4, UseOwnerScale=true)
   &AddVelocity(
    Cone Angle=$Emitter.Rand A
    Cone Axis=<0 0 1>
    Module.UseConeFalloff=false
    Module.UseVelDistribution=false
    Override Seed=false
    Velocity Mode=::NewEnumerator2
    Velocity Speed=&RandomRangeFloat(Minimum=1000, Maximum=400)
   )
   &SystemLocation(
    MaskSpawnGroup=false
    Offset=&Add_Vector(
     A=Linked() {linkedVariable {name=Particles.Position, type=NiagaraPosition}}
     B=&SelectVectorFromArray(Direct Array Index=$Particles.SpawnGroup, Vector Selection Array=DataInterface() {propertyValues {FloatData [<0 65 0>, <0 -65 0>]}})
    )
   )
  ]
  particleUpdateScript [
   &ParticleState()
   &ScaleSpriteSize(Uniform Curve Sprite Scale=DataInterface() {propertyValues {Curve {keys [{tangentMode=RCTM_User, time=0.7, value=1}, {tangentMode=RCTM_User, time=1, value=0}]}}})
   &Drag(Drag=&RandomRangeFloat(Minimum=5, Maximum=8), Use Linear Drag=true, Use Rotational Drag=true)
   &CurlNoiseForce(
    Curl Noise Cone Mask Angle=30
    Curl Noise Cone Mask Falloff Angle=30
    Mask Curl Noise=true
    Noise Strength=&FloatFromCurve(
     FloatCurve=DataInterface() {propertyValues {Curve {keys [{arriveTangent=1.0033, interpMode=RCIM_Cubic, leaveTangent=1.0033, tangentMode=RCTM_User, time=0, value=0}, <at 1 1 RCIM_Cubic>]}}}
     Scale Curve=10000
    )
    Pan Noise=false
   )
   &SolveForcesAndVelocity(Clamp Velocity=false, Limit Acceleration=false)
   &ScaleColor(
    Scale Alpha=&FNTimeOfDayScalar(Night Scale=0.7, Morning Scale=1, Evening Scale=1)
    Scale RGB=&VectorFromFloat(Value=&FN_FrontEndFloatModifier(Gameplay=1, Front End=0.05))
    ScaleA=true
    ScaleRGB=true
   )
   &Color(Color=&MultiplyLinearColor(
     Linear Color=&ColorFromCurve(ColorCurve=DataInterface() {propertyValues {
        AlphaCurve {keys [<at 0 1 RCIM_Cubic>,<at 1 1 RCIM_Cubic>]}
        BlueCurve {keys [
          {arriveTangent=-81.6115, interpMode=RCIM_Cubic, leaveTangent=-81.6115, tangentMode=RCTM_User, time=0.2409, value=30}
          <at 0.7265 0 RCIM_Cubic>
         ]}
        GreenCurve {keys [
          {arriveTangent=-211.5511, interpMode=RCIM_Cubic, leaveTangent=-211.5511, tangentMode=RCTM_User, time=0.2409, value=100.8333}
          <at 0.7265 3.3333 RCIM_Cubic>
         ]}
        RedCurve {keys [
          {arriveTangent=-399.4506, interpMode=RCIM_Cubic, leaveTangent=-399.4506, tangentMode=RCTM_User, time=0.2409, value=200}
          <at 0.7265 25 RCIM_Cubic>
         ]}
       }})
     Scale Factor=&MakeFloat4FromVectorAndFloat(
      "Vector (RGB)"=&VectorFromFloat(Value=&FN_FrontEndFloatModifier(Gameplay=1, Front End=0.05))
      "Float (Alpha)"=&FNTimeOfDayScalar(Night Scale=0.7, Morning Scale=1, Evening Scale=1)
     )
    ))
  ]
  rendererClasses [&NiagaraSpriteRendererProperties]
  renderers [{rendererClass=&NiagaraSpriteRendererProperties}]
 }
 Firework_Burst005 {
  emitterUpdateScript [
   &EmitterState(Delay First Loop Only=true, Life Cycle Mode=::NewEnumerator1, Loop Delay=0.2, Loop Duration=2, UseLoopDelay=true)
   &SpawnBurst_Instantaneous(Spawn Count=37, Use Spawn Probability=false, UseLoopCountLimit=false)
   &SpawnBurst_Instantaneous(Spawn Count=10, Spawn Time=0.2, Use Spawn Probability=false, UseLoopCountLimit=false) {module=SpawnBurst_Instantaneous001}
  ]
  particleSpawnScript [
   &FN_SetSpawnGroup(Number of Spawn Groups=2)
   &InitializeParticle(
    Color=<rgba 200 100.8333 30 1>
    Lifetime Max=0.9
    Lifetime Min=0.35
    Lifetime Mode=::NewEnumerator1
    Sprite Size Max=<2 3>
    Sprite Size Min=<1.5 4>
    Sprite Size Mode=::NewEnumerator2
    UsePositionOffset=false
    Write Material Random=true
   )
   &ShapeLocation(MaskSpawnGroup=false, Override Seed=false, Sphere Radius=4, UseOwnerScale=true)
   &AddVelocity(
    Cone Angle=120
    Cone Axis=<0 0 1>
    Inner Cone Angle=160
    Module.UseConeFalloff=false
    Module.UseVelDistribution=false
    Override Seed=false
    Velocity Mode=::NewEnumerator2
    Velocity Speed=&RandomRangeFloat(Minimum=50, Maximum=400)
   )
   &SystemLocation(
    MaskSpawnGroup=false
    Offset=&Add_Vector(
     A=Linked() {linkedVariable {name=Particles.Position, type=NiagaraPosition}}
     B=&SelectVectorFromArray(Direct Array Index=$Particles.SpawnGroup, Vector Selection Array=DataInterface() {propertyValues {FloatData [<0 65 0>, <0 -65 0>]}})
    )
   )
  ]
  particleUpdateScript [
   &ParticleState()
   &ScaleSpriteSize(Uniform Curve Sprite Scale=DataInterface() {propertyValues {Curve {keys [{arriveTangent=-5.3877, interpMode=RCIM_Cubic, leaveTangent=-5.3877, tangentMode=RCTM_User, time=0.5, value=1}, <at 1 0 RCIM_Cubic>]}}})
   &Drag(Drag=&RandomRangeFloat(Minimum=5, Maximum=8), Use Linear Drag=true, Use Rotational Drag=true)
   &CurlNoiseForce(
    Noise Frequency=100
    Noise Strength=&FloatFromCurve(FloatCurve=DataInterface() {propertyValues {Curve {keys [<at 0 0>,<at 1 1>]}}}, Scale Curve=500)
    Pan Noise=false
   ) {enabled=false}
   &GravityForce()
   &Collision(
    Age Colliding Particles=false
    CPU Trace Vector Length Multiplier=10
    KillOnCollision=false
    Maximum Penetration Correction Distance=2
    Module.Control Rest State Via Surface Normal=false
    Module.EnableMaxCollisionCount=false
    Module.Randomize Collision Normal Vector=false
    Radius Calculation Type=::NewEnumerator3
   )
   &SolveForcesAndVelocity(Clamp Velocity=false, Limit Acceleration=false)
   &CalculateAccurateVelocity()
   &ScaleMeshSizeBySpeed(Max Scale Factor=<0.1 0.1 0.1>, Min Scale Factor:Vector3f=<0.1 0.1 0.1>, Speed Threshold=1000)
   &ScaleColor(
    Scale Alpha=&FNTimeOfDayScalar(Night Scale=0.7, Morning Scale=1, Evening Scale=1)
    Scale RGB=&VectorFromFloat(Value=&FN_FrontEndFloatModifier(Gameplay=1, Front End=0.05))
    ScaleA=true
    ScaleRGB=true
   )
   &InterpolateOverTime(
    Rate Of Change:NiagaraFloat=&Divide_Float(
     A=1
     B=&FloatFromCurve(FloatCurve=DataInterface() {propertyValues {Curve {keys [<at 0.5 0.05>,<at 0.8 1>]}}})
    )
    Target Vector Value:Vector3f=$Particles.Velocity
    Variable Type=::NewEnumerator2
   )
   &Color(Color=&MultiplyLinearColor(
     Linear Color=&ColorFromCurve(ColorCurve=DataInterface() {propertyValues {
        AlphaCurve {keys [<at 0 1 RCIM_Cubic>,<at 1 1 RCIM_Cubic>]}
        BlueCurve {keys [{arriveTangent=-81.6115, interpMode=RCIM_Cubic, leaveTangent=-81.6115, tangentMode=RCTM_User, time=0, value=30}, <at 0.5092 0 RCIM_Cubic>]}
        GreenCurve {keys [
          {arriveTangent=-211.5511, interpMode=RCIM_Cubic, leaveTangent=-211.5511, tangentMode=RCTM_User, time=0, value=100.8333}
          <at 0.5092 0.4 RCIM_Cubic>
         ]}
        RedCurve {keys [{arriveTangent=-399.4506, interpMode=RCIM_Cubic, leaveTangent=-399.4506, tangentMode=RCTM_User, time=0, value=200}, <at 0.5092 3 RCIM_Cubic>]}
       }})
     Scale Factor=&MakeFloat4FromVectorAndFloat(
      "Vector (RGB)"=&VectorFromFloat(Value=&FN_FrontEndFloatModifier(Gameplay=1, Front End=0.05))
      "Float (Alpha)"=&FNTimeOfDayScalar(Night Scale=0.7, Morning Scale=1, Evening Scale=1)
     )
    ))
  ]
  rendererClasses [&NiagaraSpriteRendererProperties]
  renderers [{rendererClass=&NiagaraSpriteRendererProperties}]
 }
 Firework_Burst_Source003 {
  emitterUpdateScript [
   &EmitterState(Delay First Loop Only=true, Life Cycle Mode=::NewEnumerator1, Loop Delay=0.2, Loop Duration=2, UseLoopDelay=true)
   &SpawnBurst_Instantaneous(Spawn Count=2, Use Spawn Probability=false, UseLoopCountLimit=false)
  ]
  particleSpawnScript [
   &FN_SetSpawnGroup(Number of Spawn Groups=2)
   &InitializeParticle(
    Color=<rgba 10 5.0417 1.5 1>
    Lifetime Max=0.7
    Lifetime Min=0.6
    Lifetime Mode=::NewEnumerator1
    Position Offset=<0 0 2>
    Sprite Rotation Mode=::NewEnumerator2
    Sprite Size Mode=::NewEnumerator1
    Uniform Sprite Size Max=65
    Uniform Sprite Size Min=60
    UsePositionOffset=true
    Write Material Random=true
   )
   &SystemLocation(
    MaskSpawnGroup=false
    Offset=&Add_Vector(
     A=Linked() {linkedVariable {name=Particles.Position, type=NiagaraPosition}}
     B=&SelectVectorFromArray(Direct Array Index=$Particles.SpawnGroup, Vector Selection Array=DataInterface() {propertyValues {FloatData [<0 65 0>, <0 -65 0>]}})
    )
   )
   &CameraOffset(Camera Offset Amount=3)
  ]
  particleUpdateScript [
   &ParticleState()
   &ScaleSpriteSize(Uniform Curve Sprite Scale=DataInterface() {propertyValues {Curve {keys [
        <at 0 0>
        {tangentMode=RCTM_User, time=0.02, value=1}
        <at 0.1983 0.7399>
        <at 0.3001 1.1305>
        <at 0.4357 0.7761>
        <at 0.5052 1.0074>
        <at 0.6036 0.6814>
        <at 0.7002 0.9468>
        <at 0.7544 0.7755>
        <at 0.9036 0.9968>
        {tangentMode=RCTM_User, time=1, value=0}
       ]}}})
   &SolveForcesAndVelocity(Clamp Velocity=false, Limit Acceleration=false)
   &ScaleColor(
    Scale Alpha=0.5
    Scale RGB=&VectorFromFloat(Value=&FN_FrontEndFloatModifier(Gameplay=1, Front End=0.1))
    ScaleA=true
    ScaleRGB=true
   )
   /SetVariables(Particles.Face:Vector3f=<0 0 1>, Particles.align:Vector3f=<1 0 0>) {isSetParameters=true}
   /SetVariables(
    Particles.decal Size:Vector2f=<100 100>
    Particles.decal cam offset=1
    Particles.decal color:LinearColor=&MultiplyLinearColor(
     Linear Color=$Particles.Color
     Scale Factor=&MakeFloat4FromVectorAndFloat("Vector (RGB)"=&VectorFromFloat(Value=0.05), "Float (Alpha)"=0.1)
    )
   ) {isSetParameters=true}
   &FN_Light_Attributes(
    Color=&ColorFromCurve(ColorCurve=DataInterface() {propertyValues {
       AlphaCurve {keys []}
       BlueCurve {keys [<at 0 4.375 RCIM_Cubic>]}
       GreenCurve {keys [<at 0 17.1875 RCIM_Cubic>]}
       RedCurve {keys [<at 0 30 RCIM_Cubic>]}
      }})
    Create Color Attribute=true
    Create Exponent Attribute=true
    Create Position Attribute=true
    Create Radius Attribute=true
    Exponent=2
    Position=&SelectVectorFromArray(Direct Array Index=$Particles.SpawnGroup, Vector Selection Array=DataInterface() {propertyValues {FloatData [<0 65 10>, <0 -65 10>]}})
    Radius=&FloatFromCurve(
     FloatCurve=DataInterface() {propertyValues {Curve {keys [
         {arriveTangent=1.4346, interpMode=RCIM_Cubic, leaveTangent=1.4346, tangentMode=RCTM_User, time=0, value=0.8}
         {arriveTangent=0.1, interpMode=RCIM_Cubic, leaveTangent=0.1, time=0.2, value=1}
         <at 1 0.9 RCIM_Cubic>
        ]}}}
     Scale Curve=100
    )
    Scale Color=&FloatFromCurve(
     FloatCurve=DataInterface() {propertyValues {Curve {keys [
         <at 0 0 RCIM_Cubic>
         {arriveTangent=-0.0486, interpMode=RCIM_Cubic, leaveTangent=-0.0486, tangentMode=RCTM_User, time=0.1, value=1}
         {arriveTangent=-0.8118, interpMode=RCIM_Cubic, leaveTangent=-0.8118, tangentMode=RCTM_User, time=0.7, value=0.8}
         <at 1 0 RCIM_Cubic>
        ]}}}
     Scale Curve=1.5
    )
   )
  ]
  rendererClasses [&NiagaraSpriteRendererProperties, &NiagaraLightRendererProperties]
  renderers [
   {rendererClass=&NiagaraSpriteRendererProperties}
   {rendererClass=&NiagaraLightRendererProperties, rendererIndex=1}
  ]
 }
 Light001 {
  emitterUpdateScript [
   &EmitterState(
    Inactive Response=::NewEnumerator1
    Life Cycle Mode=::NewEnumerator1
    Loop Behavior=::NewEnumerator1
    Loop Duration Mode=::NewEnumerator1
    UseLoopDelay=false
   )
   &SpawnBurst_Instantaneous(Use Spawn Probability=false, UseLoopCountLimit=false)
  ]
  particleSpawnScript [
   &FN_SetSpawnGroup()
   &InitializeParticle(Mesh Scale Mode=::NewEnumerator3, Position Offset=<0 0 200>, UsePositionOffset=true, Write Material Random=true)
   &CameraOffset()
   /SetVariables(Particles.LightExponent=2, Particles.LightRadius=230) {isSetParameters=true}
   /SetVariables(Particles.AttenuationScale=275) {isSetParameters=true}
  ]
  particleUpdateScript [
   &ParticleState(
    Kill Particles When Lifetime Has Elapsed=false
    Let Infinitely Lived Particles Die When Emitter Deactivates=true
    Loop Particles Lifetime=true
   )
   &Color(Color=&ColorFromCurve(ColorCurve=DataInterface() {propertyValues {
       AlphaCurve {keys [<at 0 0.75 RCIM_Cubic>]}
       BlueCurve {keys [<at 0.25 0.4 RCIM_Cubic>,<at 0.35 4>,<at 0.45 4>,<at 0.5 1.45>,<at 0.65 1.45>,<at 0.7 0.4>]}
       GreenCurve {keys [<at 0.25 0.7 RCIM_Cubic>,<at 0.35 1.0667>,<at 0.45 0.8833>,<at 0.5 4>,<at 0.65 4>,<at 0.7 0.7>]}
       RedCurve {keys [<at 0.25 4 RCIM_Cubic>,<at 0.35 0.8>,<at 0.45 0.6>,<at 0.5 0.6>,<at 0.65 0.6>,<at 0.7 4>]}
      }}))
   &ScaleColor(
    Color Value To Scale=$Particles.Color
    Scale Alpha=&Multiply_Float(
     A=&FloatFromCurve(
      FloatCurve=DataInterface() {propertyValues {Curve {keys [<at 0.1 1>,<at 0.2 0.6>,<at 0.3 1>,<at 0.4 0.4>,<at 0.5 1>,<at 0.7 1>,<at 0.8 0.4>,<at 1 1>]}}}
      Scale Curve=0.65
     )
     B=&Lerp_Float(A=$User.Other Alpha Preview, B=$User.Local Alpha Preview, Alpha=$NPC.FortniteNPC.EventAudio_Amplitude)
    )
    Scale RGB=&VectorFromFloat(Value=30)
    ScaleA=true
    ScaleRGB=true
   )
   /SetVariables(Particles.LightComponentColor:LinearColor=&ColorFromCurve(ColorCurve=DataInterface() {propertyValues {
       AlphaCurve {keys [<at 0 1 RCIM_Cubic>]}
       BlueCurve {keys [<at 0.25 0.1 RCIM_Cubic>,<at 0.35 1>,<at 0.45 1>,<at 0.5 0.3625>,<at 0.65 0.3625>,<at 0.7 0.1>]}
       GreenCurve {keys [<at 0.25 0.175 RCIM_Cubic>,<at 0.35 0.2667>,<at 0.45 0.2208>,<at 0.5 1>,<at 0.65 1>,<at 0.7 0.175>]}
       RedCurve {keys [<at 0.25 1 RCIM_Cubic>,<at 0.35 0.2>,<at 0.45 0.15>,<at 0.5 0.15>,<at 0.65 0.15>,<at 0.7 1>]}
      }})) {isSetParameters=true}
   /SetVariables(Particles.LightComponentIntensity:NiagaraFloat=&Multiply_Float(
     A=&FloatFromCurve(
      FloatCurve=DataInterface() {propertyValues {Curve {keys [<at 0.1 1>,<at 0.2 0.6>,<at 0.3 1>,<at 0.4 0.4>,<at 0.5 1>,<at 0.7 1>,<at 0.8 0.4>,<at 1 1>]}}}
      Scale Curve=&Multiply_Float(A=50000, B=0.65)
     )
     B=&Lerp_Float(A=$User.Other Alpha Preview, B=$User.Local Alpha Preview, Alpha=$NPC.FortniteNPC.EventAudio_Amplitude)
    )) {isSetParameters=true}
  ]
  rendererClasses [&NiagaraLightRendererProperties, &NiagaraComponentRendererProperties]
  renderers [
   {rendererClass=&NiagaraLightRendererProperties, rendererIndex=0}
   {rendererClass=&NiagaraComponentRendererProperties, rendererIndex=1}
  ]
 }
 Sparklys {
  emitterUpdateScript [
   &EmitterState(
    Inactive Response=::NewEnumerator1
    Life Cycle Mode=::NewEnumerator1
    Loop Duration=&RandomRangeFloat(Minimum=1, Maximum=0.5)
    UseLoopDelay=false
   )
   &SpawnRate(SpawnRate=8, Use Spawn Probability=false)
   &SpawnBurst_Instantaneous(
    Spawn Count=&RandomRangeInt(Minimum=1, Maximum=3)
    Spawn Time=&RandomRangeFloat(Minimum=0, Maximum=0.5)
    Use Spawn Probability=false
    UseLoopCountLimit=false
   )
  ]
  particleSpawnScript [
   &InitializeParticle(
    Color=<rgba 1 1 1 0>
    Lifetime Max=0.5
    Lifetime Min=0.3
    Lifetime Mode=::NewEnumerator1
    Sprite Rotation Mode=::NewEnumerator2
    Sprite Size Mode=::NewEnumerator1
    Uniform Sprite Size Min=40
    UsePositionOffset=false
    Write Material Random=true
   )
   &ShapeLocation(
    MaskSpawnGroup=false
    Non Uniform Scale=<1 1 1.2>
    Offset=<0 0 150>
    Offset Mode=::NewEnumerator0
    Override Seed=false
    Shape Origin=<0 0 0>
    UseOwnerScale=true
   )
   &CameraOffset(Camera Offset Amount=-100)
  ]
  particleUpdateScript [
   &ParticleState()
   &Drag(Drag=2.5, Use Linear Drag=true, Use Rotational Drag=false)
   &SolveForcesAndVelocity(Clamp Velocity=false, Limit Acceleration=false)
   &Color(Color=&HueShiftLinearColor(
     Hue=0.1
     Saturation=1
     Value=1
     Linear Color=&ColorFromCurve(
      ColorCurve=DataInterface() {propertyValues {
        AlphaCurve {keys [<at 0 0.5 RCIM_Cubic>]}
        BlueCurve {keys [<at 0.25 0.1 RCIM_Cubic>,<at 0.297 1>,<at 0.65 0.0417>,<at 0.7 0.1>]}
        GreenCurve {keys [<at 0.25 0.1 RCIM_Cubic>,<at 0.297 0.0833>,<at 0.65 0.5>,<at 0.7 0.1>]}
        RedCurve {keys [<at 0.25 1 RCIM_Cubic>,<at 0.297 0>,<at 0.65 0>,<at 0.7 1>]}
       }}
      CurveIndex=$Emitter.NormalizedLoopAge
     )
    )) {module=Color001}
   &ScaleColor(
    Color Value To Scale=$Particles.Color
    Scale Alpha=&FloatFromCurve(
     FloatCurve=DataInterface() {propertyValues {Curve {keys [
         {arriveTangent=9.8157, interpMode=RCIM_Cubic, leaveTangent=9.8157, tangentMode=RCTM_User, time=0, value=0}
         {arriveTangent=-1.4617, interpMode=RCIM_Cubic, leaveTangent=-1.4617, tangentMode=RCTM_User, time=0.2, value=1}
         {interpMode=RCIM_Cubic, tangentMode=RCTM_User, time=0.3, value=0}
         {interpMode=RCIM_Cubic, tangentMode=RCTM_User, time=0.5, value=0.75}
         {interpMode=RCIM_Cubic, tangentMode=RCTM_Break, time=1, value=0}
        ]}}}
     Scale Curve=0.5
    )
    ScaleA=true
    ScaleRGB=true
   )
   /SetVariables(Particles.New Position:NiagaraPosition=&Add_Vector(
     A=Linked() {linkedVariable {name=Particles.Position, type=NiagaraPosition}}
     B=Linked() {linkedVariable {name=Engine.Owner.Position, type=NiagaraPosition}}
    )) {isSetParameters=true}
  ]
  rendererClasses [&NiagaraSpriteRendererProperties]
  renderers [{rendererClass=&NiagaraSpriteRendererProperties}]
 }
 Sparklys_LWE {rendererClasses [], renderers []}
 SpotLight_Main {
  emitterUpdateScript [
   &EmitterState(
    Inactive Response=::NewEnumerator1
    Life Cycle Mode=::NewEnumerator1
    Loop Behavior=::NewEnumerator1
    Loop Duration Mode=::NewEnumerator1
    UseLoopDelay=false
   )
   &SpawnBurst_Instantaneous(Use Spawn Probability=false, UseLoopCountLimit=false)
  ]
  particleSpawnScript [
   &InitializeParticle(
    Color=<rgba 1 1 1 5>
    Mesh Scale=<0.8 0.8 1.2>
    Mesh Scale Mode=::NewEnumerator4
    Position Offset=<0 0 280>
    UsePositionOffset=true
    Write Material Random=true
   )
   /SetVariables(Particles.Position002:NiagaraPosition=<0 0 3>) {isSetParameters=true}
   /SetVariables(Particles.Size002:Vector3f=<0.4 0.4 1.5>) {isSetParameters=true}
   /SetVariables(Particles.Size003:Vector3f=&VectorFromFloat(Value=0.42)) {isSetParameters=true}
   /SetVariables(Particles.Size004:Vector3f=&VectorFromFloat(Value=0.45)) {isSetParameters=true}
   &DynamicMaterialParameters(
    Index 0 Param 1=10
    Index 0 Param 2=1
    Index 0 Param 3=1
    Index 0 Param 4=2
    Index 1 Param 1=15
    Module.Param0WriteEnabled=true
    Module.Param0WriteEnabled1=true
    Module.Param1WriteEnabled=true
    Module.Param1WriteEnabled1=true
    Module.Param2WriteEnabled=true
    Module.Param2WriteEnabled1=true
    Module.Param3WriteEnabled=true
    Module.Param3WriteEnabled1=true
    Write Parameter Index 1=true
   )
  ]
  particleUpdateScript [
   &ParticleState(
    Kill Particles When Lifetime Has Elapsed=false
    Let Infinitely Lived Particles Die When Emitter Deactivates=true
    Loop Particles Lifetime=true
   )
   &Color(Color=&ColorFromCurve(ColorCurve=DataInterface() {propertyValues {
       AlphaCurve {keys [<at 0 0.3 RCIM_Cubic>]}
       BlueCurve {keys [<at 0.25 0.4 RCIM_Cubic>,<at 0.35 4>,<at 0.45 4>,<at 0.5 0.6>,<at 0.6 0.6>,<at 0.7 0.4>]}
       GreenCurve {keys [<at 0.25 1.9 RCIM_Cubic>,<at 0.35 1.0667>,<at 0.45 0.8833>,<at 0.5 0.8833>,<at 0.6 0.8833>,<at 0.7 1.9>]}
       RedCurve {keys [<at 0.25 4 RCIM_Cubic>,<at 0.35 0.8>,<at 0.45 0.6>,<at 0.5 4>,<at 0.6 4>,<at 0.7 4>]}
      }})) {module=Color001}
   &ScaleColor(
    Color Value To Scale=$Particles.Color
    Scale Alpha=&Multiply_Float(
     A=&FloatFromCurve(FloatCurve=DataInterface() {propertyValues {Curve {keys [<at 0.1 1>,<at 0.2 0.6>,<at 0.3 1>,<at 0.4 0.4>,<at 0.5 1>,<at 0.7 1>,<at 0.8 0.4>,<at 1 1>]}}})
     B=&Multiply_Float(
      A=&Lerp_Float(A=$User.Other Alpha Preview, B=$User.Local Alpha Preview, Alpha=$NPC.FortniteNPC.EventAudio_Amplitude)
      B=&FN_FrontEndFloatModifier(Gameplay=1, Front End=1.6)
     )
    )
    Scale RGB=&VectorFromFloat(Value=1.5)
    ScaleA=true
    ScaleRGB=true
   )
   /SetVariables(Particles.Color002=&MakeLinearColorFromVectorAndFloat(
     "Vector (RGB)"=&MakeVectorFromLinearColorRGB(Color=&ColorFromCurve(ColorCurve=DataInterface() {propertyValues {
         AlphaCurve {keys [<at 0 0.35 RCIM_Cubic>]}
         BlueCurve {keys [<at 0.25 0.4 RCIM_Cubic>,<at 0.35 4>,<at 0.45 4>,<at 0.5 0.08>,<at 0.6 0.08>,<at 0.7 0.4>]}
         GreenCurve {keys [<at 0.25 2.2 RCIM_Cubic>,<at 0.35 0.5167>,<at 0.45 0.5167>,<at 0.5 0.4067>,<at 0.6 0.4067>,<at 0.7 2.2>]}
         RedCurve {keys [<at 0.25 4 RCIM_Cubic>,<at 0.35 0.2>,<at 0.45 0.2>,<at 0.5 4>,<at 0.6 4>,<at 0.7 4>]}
        }}))
     "Float (Alpha)"=&Multiply_Float(
      A=&FloatFromCurve(
       FloatCurve=DataInterface() {propertyValues {Curve {keys [<at 0.1 1>,<at 0.2 0.2>,<at 0.3 1>,<at 0.4 0>,<at 0.5 1>,<at 0.7 1>,<at 0.8 0>,<at 1 1>]}}}
       Scale Curve=0.3
      )
      B=&Multiply_Float(
       A=&Lerp_Float(A=$User.Other Alpha Preview, B=$User.Local Alpha Preview, Alpha=$NPC.FortniteNPC.EventAudio_Amplitude)
       B=&FN_FrontEndFloatModifier(Gameplay=1, Front End=1.6)
      )
     )
    )) {isSetParameters=true}
   /SetVariables(Particles.Color003:LinearColor=&MakeLinearColorFromVectorAndFloat(
     "Vector (RGB)"=&MakeVectorFromLinearColorRGB(Color=&MultiplyLinearColor(
       Linear Color=<rgba 2 1.3984 0.9688 1>
       Scale Factor=&MakeVector4_FromLinearColor(LinearColor=&ColorFromCurve(ColorCurve=DataInterface() {propertyValues {
           AlphaCurve {keys [<at 0 0.75 RCIM_Cubic>]}
           BlueCurve {keys [<at 0.25 0.4 RCIM_Cubic>,<at 0.35 4>,<at 0.45 4>,<at 0.5 0.08>,<at 0.6 0.08>,<at 0.7 0.4>]}
           GreenCurve {keys [<at 0.25 1.9 RCIM_Cubic>,<at 0.35 0.5167>,<at 0.45 0.5167>,<at 0.5 0.4067>,<at 0.6 0.4067>,<at 0.7 1.9>]}
           RedCurve {keys [<at 0.25 4 RCIM_Cubic>,<at 0.35 0.2>,<at 0.45 0.2>,<at 0.5 4>,<at 0.6 4>,<at 0.7 4>]}
          }}))
      ))
     "Float (Alpha)"=&Multiply_Float(
      A=&FloatFromCurve(
       FloatCurve=DataInterface() {propertyValues {Curve {keys [<at 0.1 1>,<at 0.2 0.2>,<at 0.3 1>,<at 0.4 0>,<at 0.5 1>,<at 0.7 1>,<at 0.8 0>,<at 1 1>]}}}
       Scale Curve=0.2
      )
      B=&Multiply_Float(
       A=&Lerp_Float(A=$User.Other Alpha Preview, B=$User.Local Alpha Preview, Alpha=$NPC.FortniteNPC.EventAudio_Amplitude)
       B=&FN_FrontEndFloatModifier(Gameplay=1, Front End=6)
      )
     )
    )) {isSetParameters=true}
   /SetVariables(Particles.Color004:LinearColor=&MakeLinearColorFromVectorAndFloat(
     "Vector (RGB)"=&MakeVectorFromLinearColorRGB(Color=&MultiplyLinearColor(
       Linear Color=&ColorFromCurve(ColorCurve=DataInterface() {propertyValues {
          AlphaCurve {keys []}
          BlueCurve {keys [<at 0.25 0.4 RCIM_Cubic>,<at 0.35 4>,<at 0.45 4>,<at 0.5 0.2033>,<at 0.65 0.2033>,<at 0.7 0.4>]}
          GreenCurve {keys [<at 0.25 0.7 RCIM_Cubic>,<at 0.35 0.5167>,<at 0.45 0.5167>,<at 0.5 2>,<at 0.65 2>,<at 0.7 0.7>]}
          RedCurve {keys [<at 0.25 4 RCIM_Cubic>,<at 0.35 0.2>,<at 0.45 0.2>,<at 0.5 0.04>,<at 0.65 0.04>,<at 0.7 4>]}
         }})
       Scale Factor=&MakeVector4_FromFloat(Float=0.01)
      ))
     "Float (Alpha)"=&Multiply_Float(
      A=&FloatFromCurve(
       FloatCurve=DataInterface() {propertyValues {Curve {keys [<at 0.1 1>,<at 0.2 0.2>,<at 0.3 1>,<at 0.4 0>,<at 0.5 1>,<at 0.7 1>,<at 0.8 0>,<at 1 1>]}}}
       Scale Curve=&MakeCustomFloatFromBool(Bool=$User.bIsFrontEndPreview, True Float=0.3, False Float=0)
      )
      B=&Multiply_Float(A=70, B=&FN_FrontEndFloatModifier(Gameplay=0, Front End=1))
     )
    )) {isSetParameters=true}
  ]
  rendererClasses [&NiagaraMeshRendererProperties]
  renderers [
   {rendererClass=&NiagaraMeshRendererProperties, rendererIndex=0}
   {rendererClass=&NiagaraMeshRendererProperties, rendererIndex=1}
   {rendererClass=&NiagaraMeshRendererProperties, rendererIndex=2}
   {rendererClass=&NiagaraMeshRendererProperties, rendererIndex=3}
  ]
 }
 SpotLight_Main_LWE {rendererClasses [], renderers []}
 dome_lights {
  emitterUpdateScript [
   &EmitterState(Inactive Response=::NewEnumerator1, Life Cycle Mode=::NewEnumerator1, Loop Duration=0.75, UseLoopDelay=false)
   &SpawnRate(SpawnRate=5, Use Spawn Probability=false)
  ]
  particleSpawnScript [
   &FN_SetSpawnGroup(Number of Spawn Groups=3)
   &InitializeParticle(
    Color=<rgba 1 0.75 0.4 0.1>
    Lifetime Min=2.5
    Lifetime Mode=::NewEnumerator1
    Mesh Scale=&SelectVectorFromArray(
     Direct Array Index=$Particles.SpawnGroup
     Vector Selection Array=DataInterface() {propertyValues {FloatData [<0.6 0.6 1.4>, <0.6 0.6 1.4>, <0.6 0.6 1.4>]}}
    )
    Mesh Scale Mode=::NewEnumerator4
    Position Offset=&SelectVectorFromArray(
     Direct Array Index=$Particles.SpawnGroup
     Vector Selection Array=DataInterface() {propertyValues {FloatData [<-140 40 250>, <-50 120 250>, <-50 -120 250>]}}
    )
    UsePositionOffset=true
    Write Material Random=true
   )
   &InitialMeshOrientation(
    Mesh Orientation Mode=::NewEnumerator7
    Rotation=&SelectVectorFromArray(
     Direct Array Index=$Particles.SpawnGroup
     Vector Selection Array=DataInterface() {propertyValues {FloatData [<-0.02 -0.06 0>, <-0.06 -0.03 0>, <0.06 -0.03 0>]}}
    )
    Use Rotation Vector=true
   )
   &CameraOffset(Camera Offset Amount=-20)
  ]
  particleUpdateScript [
   &ParticleState()
   &Color(Color=&ColorFromCurve(
     ColorCurve=DataInterface() {propertyValues {
       AlphaCurve {keys [<at 0 0.75 RCIM_Cubic>]}
       BlueCurve {keys [<at 0.25 0.6 RCIM_Cubic>,<at 0.35 4>,<at 0.45 4>,<at 0.5 0.4>,<at 0.6 0.4>,<at 0.7 0.6>]}
       GreenCurve {keys [<at 0.25 2.0167 RCIM_Cubic>,<at 0.35 1.0667>,<at 0.45 0.8833>,<at 0.5 0.7>,<at 0.6 0.7>,<at 0.7 2.0167>]}
       RedCurve {keys [<at 0.25 4 RCIM_Cubic>,<at 0.35 0.8>,<at 0.45 0.6>,<at 0.5 4>,<at 0.6 4>,<at 0.7 4>]}
      }}
     CurveIndex=$Emitter.NormalizedLoopAge
    )) {module=Color001}
   &ScaleColor(
    Color Value To Scale=$Particles.Color
    Scale Alpha=&Multiply_Float(
     A=&FloatFromCurve(
      FloatCurve=DataInterface() {propertyValues {Curve {keys [
          {interpMode=RCIM_Cubic, tangentMode=RCTM_User, time=0, value=0}
          {interpMode=RCIM_Cubic, tangentMode=RCTM_User, time=0.5, value=1}
          {tangentMode=RCTM_User, time=1, value=0}
         ]}}}
      Scale Curve=0.025
     )
     B=&Multiply_Float(
      A=&Lerp_Float(A=$User.Other Alpha Preview, B=$User.Local Alpha Preview, Alpha=$NPC.FortniteNPC.EventAudio_Amplitude)
      B=&FN_FrontEndFloatModifier(Gameplay=1, Front End=2)
     )
    )
    Scale RGB=&VectorFromFloat(Value=&FloatFromCurve(
      FloatCurve=DataInterface() {propertyValues {Curve {keys [
          {interpMode=RCIM_Cubic, tangentMode=RCTM_Break, time=0.4, value=1}
          {interpMode=RCIM_Cubic, tangentMode=RCTM_Break, time=0.5, value=3}
          {interpMode=RCIM_Cubic, tangentMode=RCTM_Break, time=0.6, value=1}
          <at 0.77 1>
          <at 0.82 1>
         ]}}}
      CurveIndex=&Add_Float(
       A=$Particles.NormalizedAge
       B=&SelectFloatFromArray(
        Direct Array Index=&SpawnGroup()
        Float Selection Array=DataInterface() {propertyValues {FloatData [0.03,0.04,0.025,0.02,0.035,0.027]}}
       )
      )
     ))
    ScaleA=true
    ScaleRGB=true
   )
   &ScaleMeshSize(Scale Factor=&VectorFromFloat(Value=0.5))
  ]
  rendererClasses [&NiagaraMeshRendererProperties]
  renderers [
   {rendererClass=&NiagaraMeshRendererProperties, rendererIndex=0}
   {rendererClass=&NiagaraMeshRendererProperties, rendererIndex=1}
  ]
 }
}
summary {
 systemName=NS_Emote_Ordinary_AccusticGuitar
 emitters [
  {
   emitterName=BlueLight_MESH001
   rendererClasses [&NiagaraMeshRendererProperties, &NiagaraSpriteRendererProperties]
  }
  {emitterName=Light001, rendererClasses [&NiagaraLightRendererProperties, &NiagaraComponentRendererProperties]}
  {emitterName=Sparklys, rendererClasses [&NiagaraSpriteRendererProperties]}
  {emitterName=DarkBackGround_FE, rendererClasses [&NiagaraSpriteRendererProperties]}
  {emitterName=dome_lights, rendererClasses [&NiagaraMeshRendererProperties]}
  {emitterName=SpotLight_Main, rendererClasses [&NiagaraMeshRendererProperties]}
  {emitterName=Firework_Burst004, simTarget=GPUComputeSim, rendererClasses [&NiagaraSpriteRendererProperties]}
  {
   emitterName=Firework_Burst_Source003
   rendererClasses [&NiagaraSpriteRendererProperties, &NiagaraLightRendererProperties]
  }
  {emitterName=Firework_Burst005, rendererClasses [&NiagaraSpriteRendererProperties]}
  {emitterName=SpotLight_Main_LWE, rendererClasses []}
  {emitterName=Sparklys_LWE, rendererClasses []}
  {emitterName=DomeLight_01_LWE, rendererClasses []}
  {emitterName=DomeLight_02_LWE, rendererClasses []}
  {emitterName=DomeLight_03_LWE, rendererClasses []}
 ]
 userVariables [
  User.bIsFrontEnd {type=NiagaraBool}
  User.bIsFrontEndPreview {type=NiagaraBool}
  User.Local Alpha Preview {type=NiagaraFloat, defaultValue=0.75, description=""}
  User.Other Alpha Preview {type=NiagaraFloat, defaultValue=0.25, description=""}
 ]
}
