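# Niagara effect for the Bass Guitar emote, staging a light show with color-cycling spotlight, dome and ground beams, a light, falling glints and sparks.
# Light opacity follows flicker curves, scaled by a lerp between Other and Local preview alphas driven by an EventAudio_Amplitude parameter.
# SparkleParent003 launches a swaying spark fountain, which TrailEmbers003 samples through particle reads to spawn trails.
# Front-end modifiers adjust brightness and show a dark backdrop outside gameplay.
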NiagaraSystem /BRCosmetics/Animation/Game/MainPlayer/Emotes/Ordinary/FX/NS_Emote_Ordinary_BassGuitar
dependencies {
 usedDataInterfaces [
  &NiagaraDataInterfaceArrayFloat3
  &NiagaraDataInterfaceColorCurve
  &NiagaraDataInterfaceCurve
  &NiagaraDataInterfaceArrayFloat
  &NiagaraDataInterfaceParticleRead
 ]
 usedDynamicInputs []
 usedModules [&SystemState]
}
emitters {
 DarkBackGround_FE {
  emitterUpdateScript [
   &EmitterState(
    Inactive Response=::NewEnumerator1
    Life Cycle Mode=::NewEnumerator1
    Loop Behavior=::NewEnumerator1
    Loop Duration Mode=::NewEnumerator1
    UseLoopDelay=false
   )
   &SpawnBurst_Instantaneous(Use Spawn Probability=false, UseLoopCountLimit=false)
  ]
  particleSpawnScript [
   &FN_SetSpawnGroup()
   &InitializeParticle(
    Color=<rgba 0.01 0.0033 0 1>
    Position Offset=<0 0 80>
    Sprite Size=<250 500>
    Sprite Size Mode=::NewEnumerator4
    UsePositionOffset=true
    Write Material Random=true
   )
   &CameraOffset(Camera Offset Amount=-40)
  ]
  particleUpdateScript [
   &ParticleState(
    Kill Particles When Lifetime Has Elapsed=false
    Let Infinitely Lived Particles Die When Emitter Deactivates=true
    Loop Particles Lifetime=true
   )
   &ScaleColor(Scale Alpha=&FN_FrontEndFloatModifier(Gameplay=0, Front End=1), ScaleA=true, ScaleRGB=true)
  ]
  rendererClasses [&NiagaraSpriteRendererProperties]
  renderers [{rendererClass=&NiagaraSpriteRendererProperties}]
 }
 DomeLight_01_LWE {rendererClasses [], renderers []}
 DomeLight_02_LWE {rendererClasses [], renderers []}
 DomeLight_03_LWE {rendererClasses [], renderers []}
 Firework_Burst_Source001 {
  emitterUpdateScript [&EmitterState()]
  particleSpawnScript [
   &FN_SetSpawnGroup(Number of Spawn Groups=2)
   &InitializeParticle(
    Color=<rgba 10 3.4667 0.2 1>
    Lifetime Max=0.125
    Lifetime Min=0.1
    Lifetime Mode=::NewEnumerator1
    Position Offset=<-75 0 0>
    Sprite Rotation Mode=::NewEnumerator2
    Sprite Size Mode=::NewEnumerator1
    Uniform Sprite Size Max=50
    Uniform Sprite Size Min=45
    UsePositionOffset=true
    Write Material Random=true
   )
   &CameraOffset(Camera Offset Amount=10)
  ]
  particleUpdateScript [
   &ParticleState()
   &ScaleSpriteSize(
    Uniform Curve Scale=1.25
    Uniform Curve Sprite Scale=DataInterface() {propertyValues {Curve {keys [
        <at 0 0 RCIM_Cubic>
        {arriveTangent=-3.0037, interpMode=RCIM_Cubic, leaveTangent=-3.0037, tangentMode=RCTM_User, time=0.2, value=1}
        {tangentMode=RCTM_User, time=1, value=0}
       ]}}}
   )
   &SolveForcesAndVelocity(Clamp Velocity=false, Limit Acceleration=false)
   &ScaleColor(
    Scale RGB=&VectorFromFloat(Value=&FN_FrontEndFloatModifier(Gameplay=1, Front End=0.1))
    ScaleA=true
    ScaleRGB=true
   )
   /SetVariables(
    Particles.decal Size:Vector2f=<100 100>
    Particles.decal cam offset=1
    Particles.decal color:LinearColor=&MultiplyLinearColor(
     Linear Color=$Particles.Color
     Scale Factor=&MakeFloat4FromVectorAndFloat("Vector (RGB)"=&VectorFromFloat(Value=0.05), "Float (Alpha)"=0.1)
    )
   ) {isSetParameters=true}
   /SetVariables(Particles.Face:Vector3f=<0 0 1>, Particles.align:Vector3f=<1 0 0>) {isSetParameters=true}
  ]
  rendererClasses [&NiagaraSpriteRendererProperties]
  renderers [{rendererClass=&NiagaraSpriteRendererProperties}]
 }
 GroundLightMesh {
  emitterUpdateScript [
   &EmitterState(
    Inactive Response=::NewEnumerator1
    Life Cycle Mode=::NewEnumerator1
    Loop Behavior=::NewEnumerator1
    Loop Duration Mode=::NewEnumerator1
    UseLoopDelay=false
   )
   &SpawnBurst_Instantaneous(Spawn Count=6, Use Spawn Probability=false, UseLoopCountLimit=false)
  ]
  particleSpawnScript [
   &FN_SetSpawnGroup(Number of Spawn Groups=6)
   &InitializeParticle(
    Color=<rgba 0 0.0833 1 0.5>
    Lifetime Max=3
    Lifetime Min=1.5
    Lifetime Mode=::NewEnumerator1
    Mesh Scale=&SelectVectorFromArray(
     Direct Array Index=$Particles.SpawnGroup
     Vector Selection Array=DataInterface() {propertyValues {FloatData [<0.3 0.3 1>, <0.3 0.3 1>, <0.25 0.25 0.7>, <0.25 0.25 0.7>, <0.3 0.3 1>, <0.3 0.3 1>]}}
    )
    Mesh Scale Mode=::NewEnumerator4
    Position Offset=&Multiply_Vector(
     A=&SelectVectorFromArray(
      Direct Array Index=$Particles.SpawnGroup
      Vector Selection Array=DataInterface() {propertyValues {FloatData [<30 20 5>, <30 -20 5>, <0 35 5>, <0 -35 5>, <-40 20 5>, <-40 -20 5>]}}
     )
     B=<1.5 1.5 1>
    )
    UsePositionOffset=true
    Write Material Random=true
   )
   &InitialMeshOrientation(Mesh Orientation Mode=::NewEnumerator7, Rotation=<0.25 0.5 0>, Use Rotation Vector=true)
  ]
  particleUpdateScript [
   &ParticleState(
    Kill Particles When Lifetime Has Elapsed=false
    Let Infinitely Lived Particles Die When Emitter Deactivates=true
    Loop Particles Lifetime=true
   )
   &UpdateMeshOrientation(
    Rotation Coordinate Space=::NewEnumerator0
    Rotation Rate=&SelectFloatFromArray(Direct Array Index=$Particles.SpawnGroup, Float Selection Array=DataInterface() {propertyValues {FloatData [0.3,-0.4,0.35,-0.38,0.4,-0.35]}})
    Rotation Vector=<0 0 1>
   )
   &Color(Color=&ColorFromCurve(
     ColorCurve=DataInterface() {propertyValues {
       AlphaCurve {keys [<at 0 0.75 RCIM_Cubic>]}
       BlueCurve {keys [<at 0.1488 0.015 RCIM_Cubic>,<at 0.3893 1>,<at 0.534 1>,<at 0.7806 0.015>]}
       GreenCurve {keys [<at 0.1488 0.1375 RCIM_Cubic>,<at 0.3893 0.2875>,<at 0.534 0.2875>,<at 0.7806 0.1375>]}
       RedCurve {keys [<at 0.1488 0.75 RCIM_Cubic>,<at 0.3893 0.05>,<at 0.534 0.05>,<at 0.7806 0.75>]}
      }}
     CurveIndex=$System.NormalizedLoopAge
    )) {module=Color001}
   &ScaleColor(
    Color Value To Scale=$Particles.Color
    Scale Alpha=&Multiply_Float(
     A=&FloatFromCurve(
      FloatCurve=DataInterface() {propertyValues {Curve {keys [<at 0 1>,<at 0.2 0.5>,<at 0.4 0.89>,<at 0.5 0.45>,<at 0.7 0.8>,<at 0.9 0.7>,<at 1 1>]}}}
      CurveIndex=&Frac_Float(Float=&Add_Float(
        A=$Emitter.Age
        B=&SelectFloatFromArray(Direct Array Index=$Particles.SpawnGroup, Float Selection Array=DataInterface() {propertyValues {FloatData [0,0.3,0.5,0.4,0.77,0.18]}})
       ))
     )
     B=&Multiply_Float(
      A=&Lerp_Float(A=$User.Other Alpha Preview, B=$User.Local Alpha Preview, Alpha=$NPC.FortniteNPC.EventAudio_Amplitude)
      B=&FN_FrontEndFloatModifier(Gameplay=1, Front End=2)
     )
    )
    ScaleA=true
    ScaleRGB=true
   )
   &ScaleMeshSize(Scale Factor=&VectorFromFloat(Value=0.6))
  ]
  rendererClasses [&NiagaraMeshRendererProperties, &NiagaraSpriteRendererProperties]
  renderers [
   {rendererClass=&NiagaraMeshRendererProperties, rendererIndex=0}
   {rendererClass=&NiagaraSpriteRendererProperties, rendererIndex=1}
  ]
 }
 Light001 {
  emitterUpdateScript [
   &EmitterState(
    Inactive Response=::NewEnumerator1
    Life Cycle Mode=::NewEnumerator1
    Loop Behavior=::NewEnumerator1
    Loop Duration Mode=::NewEnumerator1
    UseLoopDelay=false
   )
   &SpawnBurst_Instantaneous(Use Spawn Probability=false, UseLoopCountLimit=false)
  ]
  particleSpawnScript [
   &FN_SetSpawnGroup()
   &InitializeParticle(Mesh Scale Mode=::NewEnumerator3, Position Offset=<0 0 200>, UsePositionOffset=true, Write Material Random=true)
   /SetVariables(Particles.AttenuationScale=275) {isSetParameters=true}
   &CameraOffset()
   /SetVariables(Particles.LightExponent=2, Particles.LightRadius=230) {isSetParameters=true}
  ]
  particleUpdateScript [
   &ParticleState(
    Kill Particles When Lifetime Has Elapsed=false
    Let Infinitely Lived Particles Die When Emitter Deactivates=true
    Loop Particles Lifetime=true
   )
   &Color(Color=&ColorFromCurve(ColorCurve=DataInterface() {propertyValues {
       AlphaCurve {keys [<at 0 0.75 RCIM_Cubic>]}
       BlueCurve {keys [<at 0.25 0.4 RCIM_Cubic>,<at 0.35 4>,<at 0.45 4>,<at 0.5 0.6>,<at 0.6 0.6>,<at 0.7 0.4>]}
       GreenCurve {keys [<at 0.25 1.9 RCIM_Cubic>,<at 0.35 1.0667>,<at 0.45 0.8833>,<at 0.5 0.8833>,<at 0.6 0.8833>,<at 0.7 1.9>]}
       RedCurve {keys [<at 0.25 4 RCIM_Cubic>,<at 0.35 0.8>,<at 0.45 0.6>,<at 0.5 4>,<at 0.6 4>,<at 0.7 4>]}
      }}))
   &ScaleColor(
    Color Value To Scale=$Particles.Color
    Scale Alpha=&Multiply_Float(
     A=&FloatFromCurve(
      FloatCurve=DataInterface() {propertyValues {Curve {keys [<at 0.1 1>,<at 0.2 0.6>,<at 0.3 1>,<at 0.4 0.4>,<at 0.5 1>,<at 0.7 1>,<at 0.8 0.4>,<at 1 1>]}}}
      Scale Curve=0.65
     )
     B=&Lerp_Float(A=$User.Other Alpha Preview, B=$User.Local Alpha Preview, Alpha=$NPC.FortniteNPC.EventAudio_Amplitude)
    )
    Scale RGB=&VectorFromFloat(Value=30)
    ScaleA=true
    ScaleRGB=true
   )
   /SetVariables(Particles.LightComponentColor:LinearColor=&ColorFromCurve(ColorCurve=DataInterface() {propertyValues {
       AlphaCurve {keys [<at 0 0.75 RCIM_Cubic>]}
       BlueCurve {keys [<at 0.25 0.1 RCIM_Cubic>,<at 0.35 1>,<at 0.45 1>,<at 0.5 0.15>,<at 0.6 0.15>,<at 0.7 0.1>]}
       GreenCurve {keys [<at 0.25 0.475 RCIM_Cubic>,<at 0.35 0.2667>,<at 0.45 0.2208>,<at 0.5 0.2208>,<at 0.6 0.2208>,<at 0.7 0.475>]}
       RedCurve {keys [<at 0.25 1 RCIM_Cubic>,<at 0.35 0.2>,<at 0.45 0.15>,<at 0.5 1>,<at 0.6 1>,<at 0.7 1>]}
      }})) {isSetParameters=true}
   /SetVariables(Particles.LightComponentIntensity:NiagaraFloat=&Multiply_Float(
     A=&FloatFromCurve(
      FloatCurve=DataInterface() {propertyValues {Curve {keys [<at 0.1 1>,<at 0.2 0.6>,<at 0.3 1>,<at 0.4 0.4>,<at 0.5 1>,<at 0.7 1>,<at 0.8 0.4>,<at 1 1>]}}}
      Scale Curve=&Multiply_Float(A=50000, B=0.65)
     )
     B=&Lerp_Float(A=$User.Other Alpha Preview, B=$User.Local Alpha Preview, Alpha=$NPC.FortniteNPC.EventAudio_Amplitude)
    )) {isSetParameters=true}
  ]
  rendererClasses [&NiagaraLightRendererProperties, &NiagaraComponentRendererProperties]
  renderers [
   {rendererClass=&NiagaraLightRendererProperties, rendererIndex=0}
   {rendererClass=&NiagaraComponentRendererProperties, rendererIndex=1}
  ]
 }
 SparkleParent003 {
  emitterUpdateScript [
   &EmitterState(Delay First Loop Only=true, Life Cycle Mode=::NewEnumerator1, Loop Delay=0.2, Loop Duration=3, UseLoopDelay=true)
   &SpawnRate(
    SpawnRate=&FloatFromCurve(
     FloatCurve=DataInterface() {propertyValues {Curve {keys [{interpMode=RCIM_Cubic, tangentMode=RCTM_User, time=0, value=2}, {interpMode=RCIM_Cubic, tangentMode=RCTM_User, time=0.35, value=0}]}}}
     CurveIndex=$Emitter.NormalizedLoopAge
     Scale Curve=8
    )
    Use Spawn Probability=false
   )
  ]
  particleSpawnScript [
   &InitializeParticle(Lifetime=0.75, Sprite Size Mode=::NewEnumerator3, Uniform Sprite Size=50, UsePositionOffset=false, Write Material Random=true)
   &AddVelocity(
    Cone Angle=1
    Cone Axis=&MakeVector(X=0, Y=&Sine(Normalized Angle=$Emitter.Age, Period=1, Scale=0.4, Bias=0), Z=1)
    Module.UseConeFalloff=false
    Module.UseVelDistribution=false
    Override Seed=false
    Velocity Mode=::NewEnumerator2
    Velocity Speed=&RandomRangeFloat(Minimum=1700, Maximum=1300)
   )
   &SystemLocation(MaskSpawnGroup=false, Offset=<-75 0 0>)
  ]
  particleUpdateScript [
   &ParticleState()
   &GravityForce(Gravity=<0 0 -300>)
   &Drag(Drag=9, Use Linear Drag=true, Use Rotational Drag=false)
   &CurlNoiseForce(Noise Frequency=120, Noise Strength=1126.3906, Pan Noise=false)
   &GravityForce(Gravity=<0 0 -32>) {module=GravityForce001}
   &SolveForcesAndVelocity(Clamp Velocity=false, Limit Acceleration=false)
   &GenerateLocationEvent(
    Event Generation Enabled:NiagaraBool=$Transient.FirstFrame
    Event Type:ENiagara_LocEventType=::NewEnumerator2
    Use Event Probability=false
    UseEventDelay=false
   )
   &DynamicMaterialParameters(
    Index 0 Param 1=0.8
    Index 0 Param 2=1
    Index 0 Param 3=1
    Index 1 Param 1=150
    Index 1 Param 4=1
    Module.Param0WriteEnabled=true
    Module.Param0WriteEnabled1=true
    Module.Param1WriteEnabled=true
    Module.Param1WriteEnabled1=true
    Module.Param2WriteEnabled=true
    Module.Param2WriteEnabled1=true
    Module.Param3WriteEnabled=false
    Module.Param3WriteEnabled1=true
    Write Parameter Index 1=true
   )
   &Color(Color=&ColorFromCurve(ColorCurve=DataInterface() {propertyValues {
       AlphaCurve {keys [<at 0 0.5 RCIM_Cubic>]}
       BlueCurve {keys [<at 0.1629 0.05 RCIM_Cubic>,<at 1 0 RCIM_Cubic>]}
       GreenCurve {keys [<at 0.1629 0.525 RCIM_Cubic>,<at 1 0 RCIM_Cubic>]}
       RedCurve {keys [<at 0.1629 1 RCIM_Cubic>,<at 1 1 RCIM_Cubic>]}
      }}))
   &ScaleColor(
    Color Value To Scale=$Particles.Color
    Scale Alpha=&FloatFromCurve(
     FloatCurve=DataInterface() {propertyValues {Curve {keys [
         {arriveTangent=9.0725, interpMode=RCIM_Cubic, leaveTangent=9.0725, tangentMode=RCTM_User, time=0, value=0}
         {arriveTangent=-7.2695, interpMode=RCIM_Cubic, leaveTangent=-7.2695, tangentMode=RCTM_User, time=0.2, value=1.75}
         {arriveTangent=-0.0405, interpMode=RCIM_Cubic, leaveTangent=-0.0405, tangentMode=RCTM_User, time=1, value=0}
        ]}}}
     Scale Curve=&Multiply_Float(
      A=0.1
      B=&FloatFromCurve(
       FloatCurve=DataInterface() {propertyValues {Curve {keys [{interpMode=RCIM_Cubic, tangentMode=RCTM_User, time=0, value=1}, {interpMode=RCIM_Cubic, tangentMode=RCTM_User, time=1, value=0}]}}}
       Scale Curve=&Waveform(
        Global Amplitude Scale=5
        "Waveform [1]"=::NewEnumerator0
        "[1] Phase Scale"=&RandomRangeFloat(Minimum=3, Maximum=5)
        "[1] Frequency"=1
        "[1] Amplitude Min/Max"=<0.5 1>
       )
      )
     )
    )
    Scale RGB=&VectorFromFloat(Value=&FN_FrontEndFloatModifier(Gameplay=25, Front End=5))
    ScaleA=true
    ScaleRGB=true
   ) {module=ScaleColor001}
   &ScaleSpriteSize(
    Uniform Curve Scale=0.25
    Uniform Curve Sprite Scale=DataInterface() {propertyValues {Curve {keys [
        {arriveTangent=2.9868, interpMode=RCIM_Cubic, leaveTangent=2.9868, tangentMode=RCTM_User, time=0, value=1.125}
        <at 0.1 2.2>
        <at 0.8 2.3>
        {arriveTangent=-0.0405, interpMode=RCIM_Cubic, leaveTangent=-0.0405, tangentMode=RCTM_User, time=1, value=1.4}
       ]}}}
   )
  ]
  rendererClasses [&NiagaraSpriteRendererProperties]
  renderers [{rendererClass=&NiagaraSpriteRendererProperties}]
 }
 Sparklys {
  emitterUpdateScript [
   &EmitterState(
    Inactive Response=::NewEnumerator1
    Life Cycle Mode=::NewEnumerator1
    Loop Duration=&RandomRangeFloat(Minimum=1, Maximum=0.5)
    UseLoopDelay=false
   )
   &SpawnRate(SpawnRate=3, Use Spawn Probability=false)
   &SpawnBurst_Instantaneous(
    Spawn Count=&RandomRangeInt(Minimum=1, Maximum=3)
    Spawn Time=&RandomRangeFloat(Minimum=0, Maximum=0.5)
    Use Spawn Probability=false
    UseLoopCountLimit=false
   )
  ]
  particleSpawnScript [
   &InitializeParticle(
    Color=<rgba 1 0.7667 0.3 1>
    Lifetime Max=0.15
    Lifetime Min=0.2
    Lifetime Mode=::NewEnumerator1
    Sprite Rotation Mode=::NewEnumerator2
    Sprite Size Mode=::NewEnumerator1
    Uniform Sprite Size Max=75
    Uniform Sprite Size Min=35
    UsePositionOffset=false
    Write Material Random=true
   )
   &ShapeLocation(
    MaskSpawnGroup=false
    Non Uniform Scale=<1 1 1.2>
    Offset=<0 0 150>
    Offset Mode=::NewEnumerator0
    Override Seed=false
    Shape Origin=<0 0 0>
    Sphere Radius=50
    UseOwnerScale=true
   )
   &CameraOffset(Camera Offset Amount=-100)
   /SetVariables(Particles.New Visibility:NiagaraInt32=&MakeIntFromFloat(Float=&UniformAOrBFloat(A=0, Distribution Weight=0.3))) {isSetParameters=true}
  ]
  particleUpdateScript [
   &ParticleState()
   &Drag(Drag=2.5, Use Linear Drag=true, Use Rotational Drag=false)
   &SolveForcesAndVelocity(Clamp Velocity=false, Limit Acceleration=false)
   &ScaleColor(
    Scale Alpha=&FloatFromCurve(FloatCurve=DataInterface() {propertyValues {Curve {keys [
         {arriveTangent=9.8157, interpMode=RCIM_Cubic, leaveTangent=9.8157, tangentMode=RCTM_User, time=0, value=0}
         {arriveTangent=-1.4617, interpMode=RCIM_Cubic, leaveTangent=-1.4617, tangentMode=RCTM_User, time=0.2, value=1}
         {interpMode=RCIM_Cubic, tangentMode=RCTM_User, time=0.3, value=0}
         {interpMode=RCIM_Cubic, tangentMode=RCTM_User, time=0.5, value=0.75}
         {interpMode=RCIM_Cubic, tangentMode=RCTM_Break, time=1, value=0}
        ]}}})
    ScaleA=true
    ScaleRGB=true
   )
   &ScaleSpriteSize(Uniform Curve Sprite Scale=DataInterface() {propertyValues {Curve {keys [{tangentMode=RCTM_User, time=0, value=1}, {tangentMode=RCTM_User, time=1, value=0}]}}})
   /SetVariables(
    Particles.New Position:NiagaraPosition=&Add_Vector(
     A=Linked() {linkedVariable {name=Particles.Position, type=NiagaraPosition}}
     B=Linked() {linkedVariable {name=Engine.Owner.Position, type=NiagaraPosition}}
    )
    Particles.New Rotation=45
    Particles.New Size:Vector2f=&Multiply_Vector2D(A=$Particles.SpriteSize, B=<0.5 0.5>)
   ) {isSetParameters=true}
  ]
  rendererClasses [&NiagaraSpriteRendererProperties]
  renderers [{rendererClass=&NiagaraSpriteRendererProperties}]
 }
 Sparklys001 {
  emitterUpdateScript [
   &EmitterState(
    Inactive Response=::NewEnumerator1
    Life Cycle Mode=::NewEnumerator1
    Loop Duration=&RandomRangeFloat(Minimum=1, Maximum=0.5)
    UseLoopDelay=false
   )
   &SpawnRate(SpawnRate=5, Use Spawn Probability=false)
  ]
  particleSpawnScript [
   &InitializeParticle(
    Color=<rgba 1 0.3325 0.0183 1>
    Lifetime Min=1.5
    Lifetime Mode=::NewEnumerator1
    Sprite Size Max=<10 50>
    Sprite Size Min=<15 35>
    Sprite Size Mode=::NewEnumerator2
    UsePositionOffset=false
    Write Material Random=true
   )
   &AddVelocity(Velocity=&RandomRangeVector(Minimum=<0 0 -50>, Maximum=<1 1 -100>))
   &SystemLocation(MaskSpawnGroup=false, Offset=<0 0 250>)
   &ShapeLocation(Box Size=<100 100 50>, MaskSpawnGroup=false, Override Seed=false, Shape Primitive=::NewEnumerator2, Surface Only=false, UseOwnerScale=true)
   &CameraOffset(Camera Offset Amount=-100)
  ]
  particleUpdateScript [
   &ParticleState()
   &GravityForce(Gravity=&RandomRangeVector(Minimum=<0 0 -300>, Maximum=<1 1 -800>))
   &Drag(Drag=2.5, Use Linear Drag=true, Use Rotational Drag=false)
   &SolveForcesAndVelocity(Clamp Velocity=false, Limit Acceleration=false)
   &ScaleColor(
    Scale Alpha=&FloatFromCurve(FloatCurve=DataInterface() {propertyValues {Curve {keys [
         {arriveTangent=9.8157, interpMode=RCIM_Cubic, leaveTangent=9.8157, tangentMode=RCTM_User, time=0, value=0}
         {arriveTangent=-1.4617, interpMode=RCIM_Cubic, leaveTangent=-1.4617, tangentMode=RCTM_User, time=0.2, value=1}
         {interpMode=RCIM_Cubic, tangentMode=RCTM_User, time=0.3, value=0}
         {interpMode=RCIM_Cubic, tangentMode=RCTM_User, time=0.5, value=0.75}
         {interpMode=RCIM_Cubic, tangentMode=RCTM_Break, time=1, value=0}
        ]}}})
    ScaleA=true
    ScaleRGB=true
   )
   &ScaleSpriteSize(
    Uniform Curve Scale=0.5
    Uniform Curve Sprite Scale=DataInterface() {propertyValues {Curve {keys [<at 0 0>, {tangentMode=RCTM_User, time=0.3, value=1}, {tangentMode=RCTM_User, time=1, value=0}]}}}
   )
   /SetVariables(
    Particles.Color 02:LinearColor=&MultiplyLinearColor(
     Linear Color=$Particles.Color
     Scale Factor=&MakeFloat4FromVectorAndFloat("Vector (RGB)"=&VectorFromFloat(Value=2000), "Float (Alpha)"=1)
    )
    Particles.SprteSize 02:Vector2f=&Multiply_Vector2D(A=$Particles.SpriteSize, B=<0.3 0.3>)
   ) {isSetParameters=true}
  ]
  rendererClasses [&NiagaraSpriteRendererProperties]
  renderers [{rendererClass=&NiagaraSpriteRendererProperties}]
 }
 Sparklys_Falling_LWE {rendererClasses [], renderers []}
 Sparklys_LWE {rendererClasses [], renderers []}
 SpotLight_Main {
  emitterUpdateScript [
   &EmitterState(
    Inactive Response=::NewEnumerator1
    Life Cycle Mode=::NewEnumerator1
    Loop Behavior=::NewEnumerator1
    Loop Duration Mode=::NewEnumerator1
    UseLoopDelay=false
   )
   &SpawnBurst_Instantaneous(Use Spawn Probability=false, UseLoopCountLimit=false)
  ]
  particleSpawnScript [
   &InitializeParticle(
    Color=<rgba 1 1 1 5>
    Mesh Scale=<0.8 0.8 1.2>
    Mesh Scale Mode=::NewEnumerator4
    Position Offset=<0 0 280>
    UsePositionOffset=true
    Write Material Random=true
   )
   /SetVariables(Particles.Position002:NiagaraPosition=<0 0 3>) {isSetParameters=true}
   /SetVariables(Particles.Size002:Vector3f=<0.4 0.4 1.5>) {isSetParameters=true}
   /SetVariables(Particles.Size003:Vector3f=&VectorFromFloat(Value=0.42)) {isSetParameters=true}
   /SetVariables(Particles.Size004:Vector3f=&VectorFromFloat(Value=0.45)) {isSetParameters=true}
   &DynamicMaterialParameters(
    Index 0 Param 1=10
    Index 0 Param 2=1
    Index 0 Param 3=1
    Index 0 Param 4=2
    Index 1 Param 1=15
    Module.Param0WriteEnabled=true
    Module.Param0WriteEnabled1=true
    Module.Param1WriteEnabled=true
    Module.Param1WriteEnabled1=true
    Module.Param2WriteEnabled=true
    Module.Param2WriteEnabled1=true
    Module.Param3WriteEnabled=true
    Module.Param3WriteEnabled1=true
    Write Parameter Index 1=true
   )
  ]
  particleUpdateScript [
   &ParticleState(
    Kill Particles When Lifetime Has Elapsed=false
    Let Infinitely Lived Particles Die When Emitter Deactivates=true
    Loop Particles Lifetime=true
   )
   &Color(Color=&ColorFromCurve(ColorCurve=DataInterface() {propertyValues {
       AlphaCurve {keys [<at 0 0.3 RCIM_Cubic>]}
       BlueCurve {keys [<at 0.25 0.4 RCIM_Cubic>,<at 0.35 4>,<at 0.45 4>,<at 0.5 0.6>,<at 0.6 0.6>,<at 0.7 0.4>]}
       GreenCurve {keys [<at 0.25 1.9 RCIM_Cubic>,<at 0.35 1.0667>,<at 0.45 0.8833>,<at 0.5 0.8833>,<at 0.6 0.8833>,<at 0.7 1.9>]}
       RedCurve {keys [<at 0.25 4 RCIM_Cubic>,<at 0.35 0.8>,<at 0.45 0.6>,<at 0.5 4>,<at 0.6 4>,<at 0.7 4>]}
      }})) {module=Color001}
   &ScaleColor(
    Color Value To Scale=$Particles.Color
    Scale Alpha=&Multiply_Float(
     A=&FloatFromCurve(FloatCurve=DataInterface() {propertyValues {Curve {keys [<at 0.1 1>,<at 0.2 0.6>,<at 0.3 1>,<at 0.4 0.4>,<at 0.5 1>,<at 0.7 1>,<at 0.8 0.4>,<at 1 1>]}}})
     B=&Multiply_Float(
      A=&Lerp_Float(A=$User.Other Alpha Preview, B=$User.Local Alpha Preview, Alpha=$NPC.FortniteNPC.EventAudio_Amplitude)
      B=&FN_FrontEndFloatModifier(Gameplay=1, Front End=1.6)
     )
    )
    Scale RGB=&VectorFromFloat(Value=1.5)
    ScaleA=true
    ScaleRGB=true
   )
   /SetVariables(Particles.Color002=&MakeLinearColorFromVectorAndFloat(
     "Vector (RGB)"=&MakeVectorFromLinearColorRGB(Color=&ColorFromCurve(ColorCurve=DataInterface() {propertyValues {
         AlphaCurve {keys [<at 0 0.35 RCIM_Cubic>]}
         BlueCurve {keys [<at 0.25 0.4 RCIM_Cubic>,<at 0.35 4>,<at 0.45 4>,<at 0.5 0.08>,<at 0.6 0.08>,<at 0.7 0.4>]}
         GreenCurve {keys [<at 0.25 2.2 RCIM_Cubic>,<at 0.35 0.5167>,<at 0.45 0.5167>,<at 0.5 0.4067>,<at 0.6 0.4067>,<at 0.7 2.2>]}
         RedCurve {keys [<at 0.25 4 RCIM_Cubic>,<at 0.35 0.2>,<at 0.45 0.2>,<at 0.5 4>,<at 0.6 4>,<at 0.7 4>]}
        }}))
     "Float (Alpha)"=&Multiply_Float(
      A=&FloatFromCurve(
       FloatCurve=DataInterface() {propertyValues {Curve {keys [<at 0.1 1>,<at 0.2 0.2>,<at 0.3 1>,<at 0.4 0>,<at 0.5 1>,<at 0.7 1>,<at 0.8 0>,<at 1 1>]}}}
       Scale Curve=0.3
      )
      B=&Multiply_Float(
       A=&Lerp_Float(A=$User.Other Alpha Preview, B=$User.Local Alpha Preview, Alpha=$NPC.FortniteNPC.EventAudio_Amplitude)
       B=&FN_FrontEndFloatModifier(Gameplay=1, Front End=1.6)
      )
     )
    )) {isSetParameters=true}
   /SetVariables(Particles.Color003:LinearColor=&MakeLinearColorFromVectorAndFloat(
     "Vector (RGB)"=&MakeVectorFromLinearColorRGB(Color=&MultiplyLinearColor(
       Linear Color=<rgba 2 1.3984 0.9688 1>
       Scale Factor=&MakeVector4_FromLinearColor(LinearColor=&ColorFromCurve(ColorCurve=DataInterface() {propertyValues {
           AlphaCurve {keys [<at 0 0.75 RCIM_Cubic>]}
           BlueCurve {keys [<at 0.25 0.4 RCIM_Cubic>,<at 0.35 4>,<at 0.45 4>,<at 0.5 0.08>,<at 0.6 0.08>,<at 0.7 0.4>]}
           GreenCurve {keys [<at 0.25 1.9 RCIM_Cubic>,<at 0.35 0.5167>,<at 0.45 0.5167>,<at 0.5 0.4067>,<at 0.6 0.4067>,<at 0.7 1.9>]}
           RedCurve {keys [<at 0.25 4 RCIM_Cubic>,<at 0.35 0.2>,<at 0.45 0.2>,<at 0.5 4>,<at 0.6 4>,<at 0.7 4>]}
          }}))
      ))
     "Float (Alpha)"=&Multiply_Float(
      A=&FloatFromCurve(
       FloatCurve=DataInterface() {propertyValues {Curve {keys [<at 0.1 1>,<at 0.2 0.2>,<at 0.3 1>,<at 0.4 0>,<at 0.5 1>,<at 0.7 1>,<at 0.8 0>,<at 1 1>]}}}
       Scale Curve=0.2
      )
      B=&Multiply_Float(
       A=&Lerp_Float(A=$User.Other Alpha Preview, B=$User.Local Alpha Preview, Alpha=$NPC.FortniteNPC.EventAudio_Amplitude)
       B=&FN_FrontEndFloatModifier(Gameplay=1, Front End=6)
      )
     )
    )) {isSetParameters=true}
   /SetVariables(Particles.Color004:LinearColor=&MakeLinearColorFromVectorAndFloat(
     "Vector (RGB)"=&MakeVectorFromLinearColorRGB(Color=&MultiplyLinearColor(
       Linear Color=&ColorFromCurve(ColorCurve=DataInterface() {propertyValues {
          AlphaCurve {keys []}
          BlueCurve {keys [<at 0.25 0.4 RCIM_Cubic>,<at 0.35 4>,<at 0.45 4>,<at 0.5 0.2033>,<at 0.65 0.2033>,<at 0.7 0.4>]}
          GreenCurve {keys [<at 0.25 0.7 RCIM_Cubic>,<at 0.35 0.5167>,<at 0.45 0.5167>,<at 0.5 2>,<at 0.65 2>,<at 0.7 0.7>]}
          RedCurve {keys [<at 0.25 4 RCIM_Cubic>,<at 0.35 0.2>,<at 0.45 0.2>,<at 0.5 0.04>,<at 0.65 0.04>,<at 0.7 4>]}
         }})
       Scale Factor=&MakeVector4_FromFloat(Float=0.01)
      ))
     "Float (Alpha)"=&Multiply_Float(
      A=&FloatFromCurve(
       FloatCurve=DataInterface() {propertyValues {Curve {keys [<at 0.1 1>,<at 0.2 0.2>,<at 0.3 1>,<at 0.4 0>,<at 0.5 1>,<at 0.7 1>,<at 0.8 0>,<at 1 1>]}}}
       Scale Curve=&MakeCustomFloatFromBool(Bool=$User.bIsFrontEndPreview, True Float=0.3, False Float=0)
      )
      B=&Multiply_Float(A=70, B=&FN_FrontEndFloatModifier(Gameplay=0, Front End=1))
     )
    )) {isSetParameters=true}
  ]
  rendererClasses [&NiagaraMeshRendererProperties]
  renderers [
   {rendererClass=&NiagaraMeshRendererProperties, rendererIndex=0}
   {rendererClass=&NiagaraMeshRendererProperties, rendererIndex=1}
   {rendererClass=&NiagaraMeshRendererProperties, rendererIndex=2}
   {rendererClass=&NiagaraMeshRendererProperties, rendererIndex=3}
  ]
 }
 SpotLight_Main_LWE {rendererClasses [], renderers []}
 TrailEmbers003 {
  emitterUpdateScript [
   &EmitterState()
   &SpawnParticlesFromOtherEmitter(
    Attribute Reader=DataInterface() {propertyValues {EmitterBinding {bindingMode=Other, emitterName=SparkleParent003}}}
    Maximum Age Spawn Threshold=2
    Minimum Age Spawn Threshold=0
    Spawn Rate=200
    Spawn Rate Per Particle Cap=500
    Use Maximum Age=false
    Use Minimum Age=false
   )
  ]
  particleSpawnScript [
   &InitializeParticle(
    Lifetime=&Lerp_Float(
     A=0.35
     B=3
     Alpha=&FloatFromCurve(
      FloatCurve=DataInterface() {propertyValues {Curve {keys [
          {
           arriveTangentWeight=1.0235
           interpMode=RCIM_Cubic
           leaveTangentWeight=1.0235
           tangentMode=RCTM_User
           tangentWeightMode=RCTWM_WeightedBoth
           time=0
           value=0
          }
          {
           arriveTangent=7.6705
           arriveTangentWeight=0.8322
           interpMode=RCIM_Cubic
           leaveTangent=7.6705
           leaveTangentWeight=0.8322
           tangentMode=RCTM_User
           tangentWeightMode=RCTWM_WeightedBoth
           time=0.9
           value=1
          }
         ]}}}
      CurveIndex=&RandomRangeFloat(Minimum=0, Maximum=1)
     )
    )
    Sprite Rotation Mode=::NewEnumerator2
    Sprite Size Mode=::NewEnumerator1
    UsePositionOffset=false
    Write Material Random=true
   )
   &SampleParticlesFromOtherEmitter(
    Applied Velocity Scale:NiagaraFloat=&FloatFromCurve(FloatCurve=DataInterface() {propertyValues {Curve {keys [<at 0 1>,<at 1 0>]}}}, Scale Curve=0.3)
    Maximum Sampled Particle Age=2
    Minimum Sampled Particle Age=0
    Module.MaxSampledAgeCutoff=true
    Module.MinSampledAgeCutoff=true
    Reject Particle By Sampled Age=true
    Velocity Sampling=::NewEnumerator0
   )
   &SphereLocation(
    MaskSpawnGroup=false
    Override Seed=false
    Sphere Radius=2
    Surface Only=false
    UseNonUniformScale=false
    UseOffset=false
    UseSphereOrientation=false
   )
  ]
  particleUpdateScript [
   &ParticleState()
   &Drag(Drag=5, Use Linear Drag=true, Use Rotational Drag=false)
   &GravityForce(Gravity=<0 0 -35>)
   &CurlNoiseForce(
    Noise Frequency=25
    Noise Strength=&FloatFromCurve(
     FloatCurve=DataInterface() {propertyValues {Curve {keys [{arriveTangent=2.3107, interpMode=RCIM_Cubic, leaveTangent=2.3107, tangentMode=RCTM_User, time=0, value=0}, <at 1 1 RCIM_Cubic>]}}}
     Scale Curve=120
    )
    Pan Noise=false
   )
   &SolveForcesAndVelocity(Clamp Velocity=false, Limit Acceleration=false)
   &Color(Color=&ColorFromCurve(ColorCurve=DataInterface() {propertyValues {
       AlphaCurve {keys [<at 0 1 RCIM_Cubic>]}
       BlueCurve {keys [<at 0.0079 3>,<at 0.1199 0>,<at 0.1889 0>,<at 0.3054 1>,<at 1 1 RCIM_Cubic>]}
       GreenCurve {keys [<at 0.0079 21>,<at 0.1199 0.4167>,<at 0.1889 0>,<at 0.3054 0>,<at 1 0 RCIM_Cubic>]}
       RedCurve {keys [<at 0.0079 30>,<at 0.1199 1>,<at 0.1889 1>,<at 0.3054 0>,<at 1 0.6257 RCIM_Cubic>]}
      }}))
   &ScaleColor(
    Color Value To Scale=$Particles.Color
    Scale Alpha=&FloatFromCurve(
     FloatCurve=DataInterface() {propertyValues {Curve {keys [
         {arriveTangent=9.0725, interpMode=RCIM_Cubic, leaveTangent=9.0725, tangentMode=RCTM_User, time=0, value=0}
         {arriveTangent=-7.2695, interpMode=RCIM_Cubic, leaveTangent=-7.2695, tangentMode=RCTM_User, time=0.2, value=1.75}
         {arriveTangent=-0.0405, interpMode=RCIM_Cubic, leaveTangent=-0.0405, tangentMode=RCTM_User, time=1, value=0}
        ]}}}
     Scale Curve=&Multiply_Float(
      A=0.1
      B=&FloatFromCurve(
       FloatCurve=DataInterface() {propertyValues {Curve {keys [{interpMode=RCIM_Cubic, tangentMode=RCTM_User, time=0, value=1}, {interpMode=RCIM_Cubic, tangentMode=RCTM_User, time=1, value=0}]}}}
       Scale Curve=&Waveform(
        Global Amplitude Scale=5
        "Waveform [1]"=::NewEnumerator0
        "[1] Phase Scale"=&RandomRangeFloat(Minimum=3, Maximum=10)
        "[1] Frequency"=1
        "[1] Amplitude Min/Max"=<0 1>
       )
      )
     )
    )
    Scale RGB=&Multiply_VectorByFloat(Vector=<25 25 25>, Float=&FN_FrontEndFloatModifier(Gameplay=1, Front End=0.15))
    ScaleA=true
    ScaleRGB=true
   ) {module=ScaleColor001}
   &ScaleSpriteSize(
    Uniform Curve Scale=0.075
    Uniform Curve Sprite Scale=DataInterface() {propertyValues {Curve {keys [
        {arriveTangent=2.9868, interpMode=RCIM_Cubic, leaveTangent=2.9868, tangentMode=RCTM_User, time=0, value=1.125}
        <at 0.4 2.3>
        {arriveTangent=-0.0405, interpMode=RCIM_Cubic, leaveTangent=-0.0405, tangentMode=RCTM_User, time=1, value=1.4}
       ]}}}
   )
  ]
  rendererClasses [&NiagaraSpriteRendererProperties]
  renderers [{rendererClass=&NiagaraSpriteRendererProperties}]
 }
 dome_lights {
  emitterUpdateScript [
   &EmitterState(Inactive Response=::NewEnumerator1, Life Cycle Mode=::NewEnumerator1, Loop Duration=0.75, UseLoopDelay=false)
   &SpawnRate(SpawnRate=5, Use Spawn Probability=false)
  ]
  particleSpawnScript [
   &FN_SetSpawnGroup(Number of Spawn Groups=3)
   &InitializeParticle(
    Color=<rgba 1 0.75 0.4 0.1>
    Lifetime Min=2.5
    Lifetime Mode=::NewEnumerator1
    Mesh Scale=&SelectVectorFromArray(
     Direct Array Index=$Particles.SpawnGroup
     Vector Selection Array=DataInterface() {propertyValues {FloatData [<0.6 0.6 1.4>, <0.6 0.6 1.4>, <0.6 0.6 1.4>]}}
    )
    Mesh Scale Mode=::NewEnumerator4
    Position Offset=&SelectVectorFromArray(
     Direct Array Index=$Particles.SpawnGroup
     Vector Selection Array=DataInterface() {propertyValues {FloatData [<-140 40 250>, <-50 120 250>, <-50 -120 250>]}}
    )
    UsePositionOffset=true
    Write Material Random=true
   )
   &InitialMeshOrientation(
    Mesh Orientation Mode=::NewEnumerator7
    Rotation=&SelectVectorFromArray(
     Direct Array Index=$Particles.SpawnGroup
     Vector Selection Array=DataInterface() {propertyValues {FloatData [<-0.02 -0.06 0>, <-0.06 -0.03 0>, <0.06 -0.03 0>]}}
    )
    Use Rotation Vector=true
   )
   &CameraOffset(Camera Offset Amount=-20)
  ]
  particleUpdateScript [
   &ParticleState()
   &Color(Color=&ColorFromCurve(
     ColorCurve=DataInterface() {propertyValues {
       AlphaCurve {keys [<at 0 0.75 RCIM_Cubic>]}
       BlueCurve {keys [<at 0.25 0.6 RCIM_Cubic>,<at 0.35 4>,<at 0.45 4>,<at 0.5 0.4>,<at 0.6 0.4>,<at 0.7 0.6>]}
       GreenCurve {keys [<at 0.25 2.0167 RCIM_Cubic>,<at 0.35 1.0667>,<at 0.45 0.8833>,<at 0.5 0.7>,<at 0.6 0.7>,<at 0.7 2.0167>]}
       RedCurve {keys [<at 0.25 4 RCIM_Cubic>,<at 0.35 0.8>,<at 0.45 0.6>,<at 0.5 4>,<at 0.6 4>,<at 0.7 4>]}
      }}
     CurveIndex=$Emitter.NormalizedLoopAge
    )) {module=Color001}
   &ScaleColor(
    Color Value To Scale=$Particles.Color
    Scale Alpha=&Multiply_Float(
     A=&FloatFromCurve(
      FloatCurve=DataInterface() {propertyValues {Curve {keys [
          {interpMode=RCIM_Cubic, tangentMode=RCTM_User, time=0, value=0}
          {interpMode=RCIM_Cubic, tangentMode=RCTM_User, time=0.5, value=1}
          {tangentMode=RCTM_User, time=1, value=0}
         ]}}}
      Scale Curve=0.025
     )
     B=&Multiply_Float(
      A=&Lerp_Float(A=$User.Other Alpha Preview, B=$User.Local Alpha Preview, Alpha=$NPC.FortniteNPC.EventAudio_Amplitude)
      B=&FN_FrontEndFloatModifier(Gameplay=1, Front End=2)
     )
    )
    Scale RGB=&VectorFromFloat(Value=&FloatFromCurve(
      FloatCurve=DataInterface() {propertyValues {Curve {keys [
          {interpMode=RCIM_Cubic, tangentMode=RCTM_Break, time=0.4, value=1}
          {interpMode=RCIM_Cubic, tangentMode=RCTM_Break, time=0.5, value=3}
          {interpMode=RCIM_Cubic, tangentMode=RCTM_Break, time=0.6, value=1}
          <at 0.77 1>
          <at 0.82 1>
         ]}}}
      CurveIndex=&Add_Float(
       A=$Particles.NormalizedAge
       B=&SelectFloatFromArray(
        Direct Array Index=&SpawnGroup()
        Float Selection Array=DataInterface() {propertyValues {FloatData [0.03,0.04,0.025,0.02,0.035,0.027]}}
       )
      )
     ))
    ScaleA=true
    ScaleRGB=true
   )
   &ScaleMeshSize(Scale Factor=&VectorFromFloat(Value=0.5))
  ]
  rendererClasses [&NiagaraMeshRendererProperties]
  renderers [
   {rendererClass=&NiagaraMeshRendererProperties, rendererIndex=0}
   {rendererClass=&NiagaraMeshRendererProperties, rendererIndex=1}
  ]
 }
}
summary {
 systemName=NS_Emote_Ordinary_BassGuitar
 emitters [
  {emitterName=dome_lights, rendererClasses [&NiagaraMeshRendererProperties]}
  {emitterName=Light001, rendererClasses [&NiagaraLightRendererProperties, &NiagaraComponentRendererProperties]}
  {
   emitterName=GroundLightMesh
   rendererClasses [&NiagaraMeshRendererProperties, &NiagaraSpriteRendererProperties]
  }
  {emitterName=Sparklys001, rendererClasses [&NiagaraSpriteRendererProperties]}
  {emitterName=DarkBackGround_FE, rendererClasses [&NiagaraSpriteRendererProperties]}
  {emitterName=Firework_Burst_Source001, rendererClasses [&NiagaraSpriteRendererProperties]}
  {emitterName=SparkleParent003, rendererClasses [&NiagaraSpriteRendererProperties]}
  {emitterName=TrailEmbers003, rendererClasses [&NiagaraSpriteRendererProperties]}
  {emitterName=SpotLight_Main, rendererClasses [&NiagaraMeshRendererProperties]}
  {emitterName=Sparklys, rendererClasses [&NiagaraSpriteRendererProperties]}
  {emitterName=DomeLight_01_LWE, rendererClasses []}
  {emitterName=DomeLight_02_LWE, rendererClasses []}
  {emitterName=DomeLight_03_LWE, rendererClasses []}
  {emitterName=SpotLight_Main_LWE, rendererClasses []}
  {emitterName=Sparklys_LWE, rendererClasses []}
  {emitterName=Sparklys_Falling_LWE, rendererClasses []}
 ]
 userVariables [
  User.bIsFrontEnd {type=NiagaraBool}
  User.bIsFrontEndPreview {type=NiagaraBool}
  User.Local Alpha Preview {type=NiagaraFloat, defaultValue=0.75, description=""}
  User.Other Alpha Preview {type=NiagaraFloat, defaultValue=0.25, description=""}
 ]
}
